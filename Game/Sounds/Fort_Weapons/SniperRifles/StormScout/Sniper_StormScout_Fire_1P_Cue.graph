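# First-person fire sound cue for the Storm Scout sniper rifle.
# A mixer layers the gunshot with a delayed, pitched-down tactical shotgun reload-end sound and a later random MilitarySniper shell casing.
# A quality-level switch plays one of three random close shots on its first two tiers and a single fixed shot on the others.
# A root modulator wraps the mix, and first-person gunfire attenuation and concurrency settings apply.
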SoundCue /Game/Sounds/Fort_Weapons/SniperRifles/StormScout/Sniper_StormScout_Fire_1P_Cue
root=Modulator_17
nodes [
 {id=Delay_0, children [Modulator_0], properties {delayMax=0.39, delayMin=0.39}}
 {id=Delay_1, children [Modulator_1], properties {delayMax=1.2, delayMin=1.1}}
 {id=Mixer_0, children [QualityLevel_1, Delay_0, Delay_1], properties {inputVolume [0.9,0.4,0.3]}}
 {id=Modulator_0, children [WavePlayer_8], properties {pitchMax=0.9, pitchMin=0.8}}
 {id=Modulator_1, children [Random_3], properties {pitchMax=1, pitchMin=1}}
 {id=Modulator_17, children [Mixer_0], properties {pitchMin=1}}
 {id=QualityLevel_1, children [Random_1,Random_1,WavePlayer_7,WavePlayer_7]}
 {id=Random_1, children [WavePlayer_1,WavePlayer_5,WavePlayer_6], properties {weights [1,1,1]}}
 {id=Random_3, children [WavePlayer_14, WavePlayer_15, WavePlayer_16], properties {weights [1,1,1]}}
 {id=WavePlayer_1, properties {soundWaveAssetPtr=&/Sniper_StormScout_Fire_Close_1P_01}}
 {id=WavePlayer_14, properties {soundWaveAssetPtr=/Game/Sounds/Fort_Weapons/SniperRifles/MilitarySniper/MilitarySniper_Casing_01}}
 {id=WavePlayer_15, properties {soundWaveAssetPtr=/Game/Sounds/Fort_Weapons/SniperRifles/MilitarySniper/MilitarySniper_Casing_02}}
 {id=WavePlayer_16, properties {soundWaveAssetPtr=/Game/Sounds/Fort_Weapons/SniperRifles/MilitarySniper/MilitarySniper_Casing_03}}
 {id=WavePlayer_5, properties {soundWaveAssetPtr=&/Sniper_StormScout_Fire_Close_1P_02}}
 {id=WavePlayer_6, properties {soundWaveAssetPtr=&/Sniper_StormScout_Fire_Close_1P_03}}
 {id=WavePlayer_7, properties {soundWaveAssetPtr=&Sniper_StormScout_Fire_Close_1P_01}}
 {id=WavePlayer_8, properties {soundWaveAssetPtr=/Game/Sounds/Fort_Weapons/Shotgun/Tactical/ShotgunTactical_ReloadEnd_01}}
]
properties {
 attenuationSettings=&Fort_Weapon_Projectile_FirstPerson_Athena_Att
 maxDistance=2097152
 pitchMultiplier=0.9
 soundClassObject=&Fort_WeaponRanged_Fire_1P_PSM
 volumeMultiplier=0.95
 attenuationOverrides {distanceAlgorithm=NaturalSound, falloffDistance=4000, attenuationShapeExtents {x=1000}}
 concurrencyOverrides {bLimitToOwner=true, maxCount=3, resolutionRule=StopOldest}
 concurrencySet [&Weapon_Gunfire_FirstPerson_Conc]
}
