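# Sound cue for the Storm Scout sniper rifle gunshot, layered by distance.
# A mixer blends close, medium and distant layers, each with its own sound class override, attenuation and slight volume modulation.
# Close shots randomly pick one of two Storm Scout waves; medium and distant layers reuse shared sniper waves.
# Those outer layers switch by quality level between random pairs and single optimized waves, and concurrency sets cap overlapping gunfire.
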SoundCue /Game/Sounds/Fort_Weapons/SniperRifles/StormScout/Sniper_StormScout_Fire_Cue
root=Mixer_1
nodes [
 {id=Attenuation_11, children [Modulator_0], properties {attenuationSettings=&Fort_Weapon_Projectile_Close_Athena_Att}}
 {
  id=Attenuation_12
  children [Modulator_18]
  properties {attenuationSettings=&Fort_Weapon_Projectile_Medium_Mono_Athena_Att}
 }
 {
  id=Attenuation_13
  children [Modulator_19]
  properties {attenuationSettings=&Fort_Weapon_Projectile_Distant_Mono_Athena_Att}
 }
 {id=Mixer_1, children [SoundClass_14, SoundClass_16, SoundClass_15], properties {inputVolume [0.9,0.6,0.3]}}
 {id=Modulator_0, children [Random_1], properties {volumeMax=2.1, volumeMin=2}}
 {id=Modulator_18, children [QualityLevel_2], properties {volumeMax=2.05, volumeMin=1.95}}
 {id=Modulator_19, children [QualityLevel_0], properties {volumeMax=2.05, volumeMin=1.95}}
 {id=QualityLevel_0, children [Random_17,Random_17,WavePlayer_0,WavePlayer_0]}
 {id=QualityLevel_2, children [Random_0,Random_0,WavePlayer_2,WavePlayer_2]}
 {id=Random_0, children [WavePlayer_3,WavePlayer_4], properties {weights [1,1]}}
 {id=Random_1, children [WavePlayer_6,WavePlayer_7], properties {weights [1,1]}}
 {id=Random_17, children [WavePlayer_38, WavePlayer_39], properties {weights [1,1]}}
 {
  id=SoundClass_14
  children [Attenuation_11]
  properties {soundClassOverride=&Fort_WeaponRanged_Fire_3P_PSM}
 }
 {
  id=SoundClass_15
  children [Attenuation_13]
  properties {soundClassOverride=&Fort_WeaponRanged_Fire_Distant}
 }
 {
  id=SoundClass_16
  children [Attenuation_12]
  properties {soundClassOverride=&Fort_WeaponRanged_Fire_Distant}
 }
 {id=WavePlayer_0, properties {soundWaveAssetPtr=/Game/Sounds/Fort_Weapons/SniperRifles/Sniper_Shot_DistantNew_A_opt}}
 {id=WavePlayer_2, properties {soundWaveAssetPtr=/Game/Sounds/Fort_Weapons/SniperRifles/Sniper_Rifle_Fire_MediumNew_New_01_opt}}
 {id=WavePlayer_3, properties {soundWaveAssetPtr=/Game/Sounds/Fort_Weapons/SniperRifles/Sniper_Rifle_Fire_MediumNew_New_01}}
 {id=WavePlayer_38, properties {soundWaveAssetPtr=/Game/Sounds/Fort_Weapons/SniperRifles/Sniper_Shot_DistantNew_A_01}}
 {id=WavePlayer_39, properties {soundWaveAssetPtr=/Game/Sounds/Fort_Weapons/SniperRifles/Sniper_Shot_DistantNew_A_02}}
 {id=WavePlayer_4, properties {soundWaveAssetPtr=/Game/Sounds/Fort_Weapons/SniperRifles/Sniper_Rifle_Fire_MediumNew_New_02}}
 {id=WavePlayer_6, properties {soundWaveAssetPtr=&/Sniper_StormScout_Fire_Close_01}}
 {id=WavePlayer_7, properties {soundWaveAssetPtr=&/Sniper_StormScout_Fire_Close_02}}
]
properties {
 maxDistance=27500
 soundClassObject=&Fort_WeaponRanged_Fire_3P_PSM
 volumeMultiplier=1
 attenuationOverrides {distanceAlgorithm=NaturalSound, falloffDistance=4000, attenuationShapeExtents {x=1000}}
 concurrencyOverrides {bLimitToOwner=true, maxCount=3, resolutionRule=StopOldest}
 concurrencySet [&Weapon_Gunfire_Local_Conc, &Weapon_Gunfire_Global_Conc]
}
