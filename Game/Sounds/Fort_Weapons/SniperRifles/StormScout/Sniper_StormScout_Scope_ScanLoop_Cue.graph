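# Sound cue for the scanning loop of the StormScout sniper rifle scope.
# A mixer layers a looping detect wave, a half-second delayed one-shot ping and a looping StormScout wave at separate input volumes.
# It plays in the weapon foley sound class with its own attenuation falloff, and its concurrency override prevents new instances rather than stopping playing ones.
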SoundCue /Game/Sounds/Fort_Weapons/SniperRifles/StormScout/Sniper_StormScout_Scope_ScanLoop_Cue
root=Mixer_0
nodes [
 {id=Delay_0, children [WavePlayer_4], properties {delayMax=0.5, delayMin=0.5}}
 {id=Mixer_0, children [WavePlayer_3,Delay_0,WavePlayer_5], properties {inputVolume [0.8,0.7,1]}}
 {
  id=WavePlayer_3
  properties {bLooping=true, soundWaveAssetPtr=&/Sniper_StormScout_Scope_DetectLoop_01}
 }
 {id=WavePlayer_4, properties {soundWaveAssetPtr=&/Sniper_StormScout_Scope_Ping_01}}
 {id=WavePlayer_5, properties {bLooping=true, soundWaveAssetPtr=&/Sniper_StormScout_Loop_01}}
]
properties {
 bOverrideAttenuation=true
 bOverrideConcurrency=true
 maxDistance=1500
 soundClassObject=&Fort_WeaponFoley
 volumeMultiplier=0.351
 attenuationOverrides {falloffDistance=1500, stereoSpread=0, attenuationShapeExtents {x=0}}
 concurrencyOverrides {resolutionRule=PreventNew}
}
