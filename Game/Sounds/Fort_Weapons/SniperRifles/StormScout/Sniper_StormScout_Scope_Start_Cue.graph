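# Sound cue for the Storm Scout sniper rifle, played as its scope starts, routed to the weapon foley sound class.
# A mixer layers the scope start wave, slightly attenuated, with a full-volume target start layer.
# The cue lowers overall pitch and volume, and uses its own attenuation and concurrency overrides, the latter preventing new instances once the limit is reached.
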SoundCue /Game/Sounds/Fort_Weapons/SniperRifles/StormScout/Sniper_StormScout_Scope_Start_Cue
root=Mixer_0
nodes [
 {id=Mixer_0, children [WavePlayer_1,WavePlayer_0], properties {inputVolume [0.8,1]}}
 {id=WavePlayer_0, properties {soundWaveAssetPtr=&/Sniper_StormScout_TargetStartLayer_01}}
 {id=WavePlayer_1, properties {soundWaveAssetPtr=&/Sniper_StormScout_Scope_Start_01}}
]
properties {
 bOverrideAttenuation=true
 bOverrideConcurrency=true
 maxDistance=2700
 pitchMultiplier=0.8
 soundClassObject=&Fort_WeaponFoley
 volumeMultiplier=0.451
 attenuationOverrides {falloffDistance=1500, nonSpatializedRadiusStart=200, attenuationShapeExtents {x=1200}}
 concurrencyOverrides {resolutionRule=PreventNew}
}
