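# Sound cue for the Storm Scout sniper rifle's scope stop, routed to the weapon foley sound class.
# A mixer layers the Weather sniper scope-stop wave with a target-start layer wave that a modulator pitches down by a fixed amount.
# The whole cue also plays at lowered pitch and volume, with its own attenuation and concurrency overrides that block new instances at the limit.
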SoundCue /Game/Sounds/Fort_Weapons/SniperRifles/StormScout/Sniper_StormScout_Scope_Stop_Cue
root=Mixer_0
nodes [
 {id=Mixer_0, children [WavePlayer_0,Modulator_0], properties {inputVolume [1,1]}}
 {id=Modulator_0, children [WavePlayer_1], properties {pitchMax=0.8, pitchMin=0.8}}
 {id=WavePlayer_0, properties {soundWaveAssetPtr=/Game/Sounds/Fort_Weapons/SniperRifles/Weather/Weather_Sniper_Scope_Stop_01}}
 {id=WavePlayer_1, properties {soundWaveAssetPtr=&/Sniper_StormScout_TargetStartLayer_01}}
]
properties {
 bOverrideAttenuation=true
 bOverrideConcurrency=true
 maxDistance=2700
 pitchMultiplier=0.8
 soundClassObject=&Fort_WeaponFoley
 volumeMultiplier=0.301
 attenuationOverrides {falloffDistance=1500, nonSpatializedRadiusStart=200, attenuationShapeExtents {x=1200}}
 concurrencyOverrides {resolutionRule=PreventNew}
}
